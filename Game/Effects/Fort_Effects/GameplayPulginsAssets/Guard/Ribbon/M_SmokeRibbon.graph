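# Unlit, two-sided translucent Niagara ribbon material that renders a panning smoke trail with a glowing energy layer.
# Emissive combines time-of-day-tinted smoke, hue-shifted center energy driven by the Frame dynamic parameter, and a refined trail texture blended by Param4.
# Opacity uses smoke alpha, particle alpha and depth fade, with a noise-distorted dissolve.
# Dynamic parameters drive dissolve and noise-based world position offset, both gated by scalars that default to off.
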Material /Game/Effects/Fort_Effects/GameplayPulginsAssets/Guard/Ribbon/M_SmokeRibbon
graph {
 properties {
  bAutomaticallySetUsageInEditor=false
  bUsedWithNiagaraRibbons=true
  blendMode=BLEND_Translucent
  shadingModel=MSM_Unlit
  twoSided=true
  usedShadingModels=Unlit
 }
 nodes [
  n1 = ParticleColor()
  n2 = VectorParameter() {defaultValue=<rgba 0 0.04 0 1>}
  n3 = AppendVector(A=@n2.R, B=@n2.R)
  n4 = VectorParameter() {
   group=Texture sample
   parameterName=Texture sample_UVControl
   channelNames {a=VPan, b=UPan, g=VScale, r=UScale}
   defaultValue=<rgba 0.5 3 -0.05 0>
  }
  n5 = StaticBoolParameter() {defaultValue=true, parameterName="[Override]ScaleUVByCenter (SB)"}
  n6 = MaterialFunctionCall!(
   Texture=TextureObject() {samplerType=SAMPLERTYPE_LinearColor, texture=&/T_EvilSmoke}
   AdditionalUVOffset=@n3
   UVControl_RGB=@n4
   UVControl_A=@n4
   "[Override]TimeInput"
   "[Override]ScaleUVByCenter"=@n5
   "[Override]UVInput"
  ) {materialFunction=/Game/Effects/Niagara/MasterMaterials/Functions/Common/FXCommon_TextureSample_V2}
  n7 = MaterialFunctionCall!(TexCoord=@n6.UV, ClampChannel=Constant2Vector() {g=1}) {materialFunction=/Game/Effects/Fort_Effects/MaterialFunctions/MF_FX_TilingMethod}
  n8 = Add(B=Multiply(A=@n6.A, B=@n7.Result))
  n9 = TextureObjectParameter() {
   group=Dissolve
   parameterName=Dissolve_Texture
   samplerType=SAMPLERTYPE_LinearColor
   sortPriority=0
   texture=/Engine/Functions/Engine_MaterialFunctions02/ExampleContent/Textures/OffsetNoiseDistanceFields_ForNormals
  }
  n10 = TextureObjectParameter() {
   group=DissolveNoise
   parameterName=DissolveNoise_Texture
   sortPriority=0
   texture=/Game/Effects/Fort_Effects/Emotes/UltraLight/Texture/T_SoftNoise_BC7
  }
  n11 = VectorParameter() {group=DissolveNoise, parameterName=DissolveNoise_UVControl, channelNames {a=VPan, b=UPan, g=VScale, r=UScale}, defaultValue=<rgba 1 1 0.2 0>}
  n12 = VectorParameter() {
   group=DissolveNoise
   parameterName=DissolveNoise_NoiseControl
   sortPriority=60
   channelNames {a=VStrength, b=UStrength, g=VOffset, r=UOffset}
   defaultValue=<rgba -0.5 -0.5 0.2 0.2>
  }
  n13 = MaterialFunctionCall!(
   NoiseValue=MaterialFunctionCall(Texture=@n10, UVControl_RGB=@n11, UVControl_A=@n11) {materialFunction=&FXCommon_TextureSample}.RG
   Control_RGB=@n12
   Control_A=@n12
  ) {materialFunction=/Game/Effects/Niagara/MasterMaterials/Functions/Common/FXCommon_NoiseControl}
  n14 = VectorParameter() {group=Dissolve, parameterName=Dissolve_UVControl, channelNames {a=VPan, b=UPan, g=VScale, r=UScale}, defaultValue=<rgba 0.5 2 -0.2 0>}
  n15 = DynamicParameter() {parameterIndex=1, defaultValue=<rgba 0 1 1 1>, paramNames [Dissolve,WPO,Param3,Param4]}
  n16 = ScalarParameter() {defaultValue=0.5, parameterName=f2}
  n17 = Subtract(
   A=Add(A=MaterialFunctionCall(Texture=@n9, AdditionalUVOffset=@n13.NoiseValue, UVControl_RGB=@n14, UVControl_A=@n14) {materialFunction=&FXCommon_TextureSample}.R)
   B=Multiply(A=Power(Base=@n15.Dissolve, Exp=@n16), B=ScalarParameter() {defaultValue=1.3, parameterName=f32})
  )
  n18 = ScalarParameter() {defaultValue=0.5, parameterName=c}
  n19 = OneMinus(None=Saturate(None=MaterialFunctionCall(In=@n17, Contrast=@n18) {materialFunction=&CheapContrast}.Result))
  n20 = ScalarParameter() {parameterName=UseDissolve}
  DepthFade(
   Opacity=Multiply(A=Multiply(A=@n1.A, B=@n8), B=LinearInterpolate(B=@n19, Alpha=@n20) {constA=1})
   FadeDistance=ScalarParameter() {parameterName=DFD}
  )
  DynamicParameter() {defaultValue=<rgba 0 1 1 1>, paramNames [Frame,Fram,Emis,Param4]}
  n21 = TextureObjectParameter() {group=WPO, parameterName=WPO_Texture, sortPriority=0, texture=&T_SoftNoise_BC7}
  n22 = VectorParameter() {group=WPO, parameterName=WPO_UVControl, channelNames {a=VPan, b=UPan, g=VScale, r=UScale}, defaultValue=<rgba 1 1 -0.5 0>}
  n23 = TextureCoordinate()
  n24 = Constant2Vector() {r=1}
  n25 = Add(A=MaterialFunctionCall(Texture=@n21, UVControl_RGB=@n22, UVControl_A=@n22, "[Override]UVInput"=Multiply(A=@n23, B=@n24)) {materialFunction=&FXCommon_TextureSample}.RGB) {constB=-0.5}
  LinearInterpolate(B=Multiply(A=Transform(None=@n25) {transformSourceType=TRANSFORMSOURCE_Local}, B=@n15.WPO), Alpha=ScalarParameter() {parameterName=UseWPO})
  n26 = MaterialFunctionCall(
   Texture=TextureObject() {samplerType=SAMPLERTYPE_LinearColor, texture=&/Trail_Particle_FC}
   AdditionalUVOffset=@n3
   UVControl_RGB=@n4
   UVControl_A=@n4
   "[Override]ScaleUVByCenter"=@n5
  ) {materialFunction=&FXCommon_TextureSample_V2}
  n27 = VectorParameter() {parameterName=Refine, defaultValue=<rgba 0.7 3 1 0.5>}
  n28 = DynamicParameter() {defaultValue=<rgba 0 1 1 1>, paramNames [Frame,Fram,Emis,Param4]}
  n29 = LinearInterpolate(
   B=Multiply(
    A=MaterialFunctionCall(Texture=@n26.R, RGB=@n27.RGB, A=@n27.RGB) {materialFunction=&MF_FX_Refine}.Result
    B=VectorParameter() {parameterName=C, defaultValue=<rgba 0.436 0 1 1>}.RGB
   )
   Alpha=@n28.Param4
  )
  n30 = TextureObjectParameter() {
   group=CenterEnergy
   parameterName=CenterEnergy_Texture
   sortPriority=0
   texture=/Game/Weapons/FORT_Melee/Pickaxe_Inferno/FX/T_Pickaxe_Inferno_Trail_01
  }
  n31 = VectorParameter() {group=CenterEnergy, parameterName=CenterEnergy_UVControl, channelNames {a=VPan, b=UPan, g=VScale, r=UScale}, defaultValue=<rgba 3 1 0.05 0>}
  n32 = ScalarParameter() {defaultValue=4, parameterName=VS}
  n33 = MaterialFunctionCall(
   Texture=@n30
   UVControl_RGB=MaterialFunctionCall(X=@n31.UScale, Y=LinearInterpolate(A=@n32, B=@n31.VScale, Alpha=@n28.Frame), Z=@n31.UScale) {materialFunction=&MakeFloat3}.Result
   UVControl_A=@n31.VPan
   "[Override]ScaleUVByCenter"=@n5
  ) {materialFunction=&FXCommon_TextureSample_V2}
  n34 = VectorParameter() {parameterName=CenterEnergyRefine2, defaultValue=<rgba 5 5 1 0.5>}
  n35 = VectorParameter() {parameterName=CenterEnergyRefine, defaultValue=<rgba 0.2 1 0.5 0.5>}
  n36 = LinearInterpolate(A=AppendVector(A=@n34.RGB, B=@n34.RGB), B=AppendVector(A=@n35.RGB, B=@n35.RGB), Alpha=@n28.Frame)
  n37 = ComponentMask(None=@n36) {b=true, g=true, r=true}
  n38 = ComponentMask(None=@n36) {a=true}
  n39 = Multiply(A=OneMinus(None=@n28.Frame), B=ScalarParameter() {defaultValue=0.5, parameterName=H})
  n40 = VectorParameter() {parameterName=Color, defaultValue=<rgba 0 70 100 1>}
  n41 = Multiply(
   A=MaterialFunctionCall(TexCoord=@n33.UV, ClampChannel=Constant2Vector() {g=1}) {materialFunction=&MF_FX_TilingMethod}.Result
   B=Multiply(
    A=MaterialFunctionCall(Texture=@n33.RGB, RGB=@n37, A=@n38) {materialFunction=&MF_FX_Refine}.Result
    B=MaterialFunctionCall(Hue Shift Percentage=@n39, Texture=@n40.RGB) {materialFunction=&HueShift}.Result
   )
  )
  n42 = MaterialFunctionCall(
   Color_Morning=VectorParameter() {parameterName="Color_Morning (V3)", defaultValue=<rgba 1 1 1 1>}.RGB
   Color_Day=VectorParameter() {parameterName="Color_Day (V3)", defaultValue=<rgba 1 1 1 1>}.RGB
   Color_Evening=VectorParameter() {parameterName="Color_Evening (V3)", defaultValue=<rgba 1 1 1 1>}.RGB
   Color_Night=VectorParameter() {parameterName="Color_Night (V3)", defaultValue=<rgba 1 1 1 1>}.RGB
  ) {materialFunction=&Color_Time_of_Day}
  n43 = LinearInterpolate(
   A=VectorParameter() {parameterName=BotColor, defaultValue=<rgba 0.1 0.1 0.1 1>}.RGB
   B=VectorParameter() {parameterName=TipColor, defaultValue=<rgba 0.198 0.14 0.2 1>}.RGB
   Alpha=@n28.Frame
  )
  n44 = Multiply(A=@n6.RGB, B=@n7.Result)
  n45 = VectorParameter() {parameterName=Power, defaultValue=<rgba 0.8 0.8 0.8 0.3>}
  n46 = Multiply(A=Power(Base=LinearInterpolate(A=@n43, B=@n44), Exp=@n45.A), B=@n45.RGB)
  n47 = Multiply(A=Add(A=Multiply(A=@n28.Frame, B=@n41), B=Multiply(A=@n42.Result, B=@n46)), B=@n1.RGB)
  MaterialFunctionCall(Emissive=Divide(A=Add(A=@n29, B=@n47))) {materialFunction=&MF_FX_EmissiveAdjustments}
  ScalarParameter() {defaultValue=5}
 ]
}
